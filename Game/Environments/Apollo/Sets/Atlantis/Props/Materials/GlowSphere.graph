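# Unlit, two-sided additive material for a glowing, glassy sphere prop in the Atlantis set.
# Two opposing panned swirl normals, dotted with the camera vector, drive a ramp-texture glass tint, added to a time-of-day-scaled, view-angle glow.
# That sum is overlay-blended with a color, then topped with a tight sky highlight from normal direction.
# Further chains build a sine bob offset and a depth- and view-angle-based fade.
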Material /Game/Environments/Apollo/Sets/Atlantis/Props/Materials/GlowSphere
graph {
 properties {
  blendMode=BLEND_Additive
  pixelDepthOffsetMode=PDOM_Legacy
  refractionMethod=RM_IndexOfRefraction
  shadingModel=MSM_Unlit
  twoSided=true
  usedShadingModels=Unlit
 }
 nodes [
  n1 = Multiply(A=TextureCoordinate(), B=ScalarParameter() {defaultValue=1.25})
  n2 = Add(
   A=TextureSample(UVs=Panner() {speedX=-0.1, speedY=-0.1}) {samplerType=SAMPLERTYPE_Normal, texture=/Game/Global/Textures/Default/Noise/T_SwirlNoise_N}.RGB
   B=TextureSample(UVs=Panner(Coordinate=@n1) {speedX=0.1, speedY=0.1}) {samplerType=SAMPLERTYPE_Normal, texture=&T_SwirlNoise_N}.RGB
  )
  n3 = ComponentMask(None=@n2) {r=true}
  n4 = ScalarParameter() {parameterName=BobSpeed}
  Add(
   A=@n3
   B=LinearInterpolate(
    B=VectorParameter() {parameterName=WPO Bob, defaultValue=<rgba 0 0 0 1>}.RGB
    Alpha=MaterialFunctionCall(Speed=@n4) {materialFunction=&DebugTimeSine}.Result
   )
  )
  n5 = VectorParameter() {parameterName=Color, defaultValue=<rgba 0.9948 0.5237 0.213 0.215>}
  Add(A=@n5.A, B=Constant() {r=0.25})
  n6 = Saturate(None=DotProduct(A=Constant3Vector() {constant=<rgba 0 0 -1 0>}, B=VertexNormalWS()))
  n7 = Constant() {r=2.3}
  n8 = Constant() {r=2}
  n9 = TwoSidedSign()
  n10 = Saturate(None=Multiply(A=Power(Base=@n6, Exp=@n7), B=Multiply(A=@n8, B=@n9)))
  n11 = CameraVectorWS()
  n12 = DotProduct(A=@n11, B=VertexNormalWS())
  n13 = ScalarParameter() {defaultValue=7.268, parameterName=Param_1}
  n14 = Saturate(None=Multiply(A=Power(Base=@n12, Exp=@n13), B=ScalarParameter() {defaultValue=1216.8301, parameterName=Param_2}))
  n15 = ScalarParameter() {defaultValue=5, parameterName=Glow Contrast}
  n16 = CollectionParameter() {
   collection=&FortniteMaterialParameters
   parameterName=FortniteActiveTimeOfDay
  }
  n17 = VectorParameter() {
   group=Actor Gradient
   parameterName=TOD Actor Grad Emissive
   channelNames {a=Night, b=Evening, g=Day, r=Morning}
   defaultValue=<rgba 0.5 0.5 0.5 1>
  }
  n18 = AppendVector(A=@n17, B=@n17)
  n19 = Multiply(
   A=Saturate(None=Power(Base=@n12, Exp=@n15))
   B=Multiply(A=DotProduct(A=@n16, B=@n18), B=VectorParameter() {parameterName=Glow, defaultValue=<rgba 0 0.741 1 0>}.RGB)
  )
  n20 = TextureSample(UVs=DotProduct(A=@n11, B=Transform(None=@n2))) {
   samplerSource=SSM_Clamp_WorldGroupSettings
   texture=/Game/Characters/Player/Male/Medium/Bodies/M_MED_Winter_Ghoul/Textures/T_Ramp_Vert_QuickIn_SoftOut
  }
  n21 = Multiply(A=@n20.R, B=VectorParameter() {parameterName=Glass, defaultValue=<rgba 0.1068 0.3078 0.487 0>}.RGB)
  Add(
   A=Multiply(A=VectorParameter() {parameterName=Sky, defaultValue=<rgba 24.6274 338.6982 555 0>}.RGB, B=Multiply(A=@n10, B=@n14))
   B=MaterialFunctionCall(Base=Add(A=@n19, B=@n21), Blend=@n5.RGB) {materialFunction=&Blend_Overlay}.Result
  )
  n22 = Constant() {r=1}
  n23 = Constant() {r=0.9}
  n24 = Divide(
   A=Subtract(
    A=MaterialFunctionCall!(World Position) {materialFunction=/Engine/Functions/Engine_MaterialFunctions02/Utility/DepthFromWorldPosition}.Result
    B=Constant() {r=1200}
   )
   B=Constant() {r=-1000}
  )
  LinearInterpolate(
   A=LinearInterpolate(B=LinearInterpolate(A=@n22, B=@n23, Alpha=@n3) {constA=1}, Alpha=Saturate(None=@n24)) {constA=1}
   Alpha=Saturate(None=@n12)
  )
 ]
}
