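# Unlit additive water glow material for Atlantis props, built from animated algae textures projected in actor-relative world XY.
# Two rotating algae layers, one warped by cloud noise, are masked by exponential radial gradients and added to a soft radial glow.
# A time-of-day weight from the Fortnite parameter collection drives gradient density and overall intensity.
# The result feeds a depth fade, alongside a teal Color parameter.
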Material /Game/Environments/Apollo/Sets/Atlantis/Props/Materials/M_AddativeWater
graph {
 properties {
  blendMode=BLEND_Additive
  pixelDepthOffsetMode=PDOM_Legacy
  previewMesh=/Game/Environments/Hera/FX/Meshes/SM_WaterTankBreak_GlassCracks_Water_01
  shadingModel=MSM_Unlit
  usedShadingModels=Unlit
 }
 nodes [
  n1 = WorldPosition()
  n2 = ActorPositionWS()
  n3 = Multiply(A=ComponentMask(None=Subtract(A=@n1.XYZ, B=@n2)) {g=true, r=true}, B=Constant() {r=0.005})
  n4 = Multiply(
   A=TextureSample(UVs=Rotator(Coordinate=@n3) {speed=0.01}) {samplerType=SAMPLERTYPE_LinearGrayscale, texture=/Game/Global/Textures/Default/PointSplatter/T_GlowAlgae}.R
   B=Constant() {r=255}
  )
  n5 = CollectionParameter() {
   collection=&FortniteMaterialParameters
   parameterName=FortniteActiveTimeOfDay
  }
  n6 = VectorParameter() {
   group=Actor Gradient
   parameterName=TOD Actor Grad Emissive
   channelNames {a=Night, b=Evening, g=Day, r=Morning}
   defaultValue=<rgba 0.5 0.5 0.5 1>
  }
  n7 = DotProduct(A=@n5, B=AppendVector(A=@n6, B=@n6))
  n8 = MaterialFunctionCall(Radius=Constant() {r=0.15}, Density=Multiply(A=Constant() {r=0.2}, B=@n7)) {materialFunction=&RadialGradientExponential}
  n9 = TextureSample(UVs=@n3) {samplerType=SAMPLERTYPE_Masks, texture=/Game/Global/Textures/Default/Noise/T_Noise_Cloud_Mask}
  n10 = Constant() {r=0.1}
  n11 = TextureSample(UVs=Add(A=Rotator(Coordinate=@n3) {speed=0.035}, B=Multiply(A=@n9.R, B=@n10))) {samplerType=SAMPLERTYPE_LinearGrayscale, texture=&T_GlowAlgae}
  n12 = Constant() {r=4}
  n13 = Constant() {r=0.35}
  n14 = Multiply(
   A=Add(A=Multiply(A=@n4, B=@n8.RadialGradientExponential), B=Multiply(A=@n11.R, B=@n12))
   B=MaterialFunctionCall(Radius=Constant() {r=0.4}, Density=Multiply(A=@n13, B=@n7)) {materialFunction=&RadialGradientExponential}.RadialGradientExponential
  )
  n15 = MaterialFunctionCall(Density=Constant() {r=0.5}) {materialFunction=&RadialGradientExponential}
  n16 = Constant() {r=15}
  DepthFade(
   Opacity=Multiply(A=Add(A=@n14, B=@n15.RadialGradientExponential), B=Multiply(A=@n16, B=@n7))
   FadeDistance=ScalarParameter() {defaultValue=506.2123, parameterName=FadeDistance}
  )
  VectorParameter() {parameterName=Color, defaultValue=<rgba 0.0051 0.1406 0.1377 1>}
 ]
}
